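# NS_Contrail_Hearts is a Niagara skydiving contrail cosmetic that trails hearts and pink effects from the player's skeletal mesh.
# Hearts_Hand spawns small sprites at two bones, Ribbons draws four bone-attached trails, and SmokeRing periodically bursts one expanding pink sprite.
# FN_DivingAttributes runs at system level, with skydive velocity driving particle motion and front-end modifiers switching values between gameplay and menus.
# The Large_Heart emitter is disabled.
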NiagaraSystem /BRCosmetics/Athena/Effects/Particles/Skydiving/NS_Contrail_Hearts/NS_Contrail_Hearts
dependencies {
 usedDataInterfaces [&NiagaraDataInterfaceCurve, &NiagaraDataInterfaceVectorCurve, &NiagaraDataInterfaceColorCurve]
 usedDynamicInputs [&Lerp_Vector, &Lerp_Float, &SetBoolByFloatComparison]
 usedModules [
  &SystemState
  /Game/Effects/Niagara/Modules/Skydiving/FN_DivingAttributes/v5/FN_DivingAttributes
  &self:SystemSpawnScript.SystemScriptSource.SystemScriptGraph.NiagaraNodeAssignment_0.SetVariables_#1
  &self:SystemSpawnScript.SystemScriptSource.SystemScriptGraph.NiagaraNodeAssignment_2.SetVariables_#2
  &self:SystemSpawnScript.SystemScriptSource.SystemScriptGraph.NiagaraNodeAssignment_1.SetVariables_#3
  &FN_Integrator
 ]
}
emitters {
 Hearts_Hand {
  emitterUpdateScript [
   &EmitterState(Inactive Response=::NewEnumerator1, Life Cycle Mode=::NewEnumerator1, Loop Duration=1, UseLoopDelay=true)
   &SpawnRate(
    SpawnRate=&RandomRangeFloat(Minimum=40, Maximum=60, Evaluation Type=::NewEnumerator1)
    Use Spawn Probability=false
   )
  ]
  particleSpawnScript [
   &FN_SetSpawnGroup(Number of Spawn Groups=2)
   /Niagara/Modules/Spawn/Initialization/InitializeParticle(
    Lifetime=&RandomRangeFloat(Minimum=0.3, Maximum=0.5, Evaluation Type=::NewEnumerator1)
    Module.Write Color=true
    Module.Write Lifetime=true
    Module.Write Mass=true
    Module.Write Material Random=true
    Module.Write Position=false
    Module.Write Scale=false
    Module.Write SpriteRotation=false
    Module.Write SpriteSize=true
    Position=<0 0 0>
    Sprite Size=&Vector2DFromFloat(Value=&RandomRangeFloat(Minimum=3, Maximum=20, Evaluation Type=::NewEnumerator1))
   )
   &Color(Color=&MakeLinearColorFromVectorAndFloat(
     "Vector (RGB)"=&VectorFromFloat(Value=&RandomRangeFloat(Minimum=1, Maximum=5, Evaluation Type=::NewEnumerator1))
     "Float (Alpha)"=&ClampFloat(Float=1.5)
    ))
   &SkeletalMeshLocation(
    Bone / Socket Index=$Particles.SpawnGroup
    Bone Sampling Mode=::NewEnumerator4
    Inherit Velocity=false
    MaskSpawnGroup=false
    Skeletal Mesh=$User.SkeletalMesh
   )
   /Niagara/Modules/Spawn/Location/SphereLocation(MaskSpawnGroup=false, Module.Override Randomness=false, Sphere Radius=10, Surface Only=false)
   &AddVelocity(Velocity=&Multiply_VectorByFloat(
     Vector=$Output.SphereLocation.SphereVector
     Float=&RandomRangeFloat(Minimum=5, Maximum=10, Evaluation Type=::NewEnumerator1)
    )) {module=AddVelocity001}
   &AddVelocity(
    Coordinate Space=::Simulation
    Velocity=&Multiply_Vector(
     A=$System.SkydiveVelocity
     B=&FN_FrontEndVector3Modifier(Gameplay=<0 0 0.9>, Front End=&VectorFromFloat(Value=-0.1))
    )
   ) {module=AddVelocity002}
   &DynamicMaterialParameters(
    Index 0 Param 2=&RandomRangeFloat(Minimum=0, Maximum=1, Evaluation Type=::NewEnumerator1)
    Index 0 Param 3=&RandomRangeFloat(Minimum=0, Maximum=2, Evaluation Type=::NewEnumerator1)
    Module.Param0WriteEnabled=false
    Module.Param1WriteEnabled=true
    Module.Param2WriteEnabled=true
    Module.Param3WriteEnabled=false
   ) {module=DynamicMaterialParameters001}
  ]
  particleUpdateScript [
   &ParticleState()
   &InheritSourceMovement(
    Applied Position Scale=&FN_FrontEndVector3Modifier(Gameplay=<0.5 0.5 0.25>, Front End=<0.4 0.4 0.2>)
    ApplyForce=false
    ApplyPosition=true
    ApplyPositionFalloff=false
    Source Velocity:Vector3f=$System.SkydiveVelocity
   )
   &ScaleColor(
    Scale Alpha=&FloatFromCurve(FloatCurve=DataInterface() {propertyValues {Curve {keys [<at 0 0 RCIM_Cubic>,<at 0.25 1>,<at 1 0>]}}})
    ScaleA=true
    ScaleRGB=true
   )
   &ScaleSpriteSize(Scale Factor=&Vector2DFromFloat(Value=&FloatFromCurve(FloatCurve=DataInterface() {propertyValues {Curve {keys [{arriveTangent=0.3, interpMode=RCIM_Cubic, time=0.2, value=1}, {interpMode=RCIM_Cubic, leaveTangent=-0.8267, time=1, value=0.3}]}}})))
   &Drag(
    Drag=&RandomRangeFloat(Minimum=3, Maximum=7, Evaluation Type=::NewEnumerator1)
    Use Linear Drag=true
    Use Rotational Drag=false
   )
   &SolveForcesAndVelocity(Clamp Velocity=false, Limit Acceleration=false)
   &DynamicMaterialParameters(
    Index 0 Param 1=&FloatFromCurve(FloatCurve=DataInterface() {propertyValues {Curve {keys [{arriveTangent=5.1264, interpMode=RCIM_Cubic, time=0, value=0}, <at 0.5 1 RCIM_Cubic>]}}})
    Index 0 Param 4=1
    Module.Param0WriteEnabled=true
    Module.Param1WriteEnabled=false
    Module.Param2WriteEnabled=false
    Module.Param3WriteEnabled=true
   )
  ]
  rendererClasses [&NiagaraSpriteRendererProperties]
  renderers [{rendererClass=&NiagaraSpriteRendererProperties}]
 }
 Large_Heart {
  bEnabled=false
  emitterUpdateScript [
   &EmitterLifeCycle(DurationRecalcEachLoop=true, NextLoopDuration:NiagaraFloat=&UniformRangedFloat(Minimum=3, Maximum=1))
   &SpawnBurst_Instantaneous(Spawn Count=1, Spawn Group=5, Use Spawn Probability=false)
  ]
  particleSpawnScript [
   &InitializeParticle(
    Color=<rgba 1 1 1 0.333>
    Lifetime=&UniformRangedFloat(Minimum=0.55, Maximum=0.75)
    Module.Write Color=true
    Module.Write Lifetime=true
    Module.Write Mass=false
    Module.Write Material Random=true
    Module.Write Position=false
    Module.Write Scale=false
    Module.Write SpriteRotation=false
    Module.Write SpriteSize=true
    Position=&SimulationPosition()
    Sprite Size=<100 100>
   )
   &SampleSkeletalMeshSkeleton(Mesh:NiagaraDataInterfaceSkeletalMesh=$User.SkeletalMesh, Specific Bone Index:NiagaraInt32=&SpawnGroup()) {module=SampleSkeletalMeshSkeleton001}
   /Niagara/Modules/Spawn/Location/SkeletalMeshSkeletonLocation(
    Module.Inherit Velocity=true
    Module.Use Sampled Bone as Ribbon ID=true
    Sampled Position Attribute:Vector3f=$Particles.SampleSkeletalMeshSkeleton001.SampledPosition
    Sampled Position Offset:Vector3f=&Add_Vector(A=&TransformVector(Vector=<0 200 -150>, Destination Space=::Simulation), B=<0 0 0>)
    Sampled Velocity Attribute:Vector3f=$Particles.SampleSkeletalMeshSkeleton001.SampledVelocity
   ) {module=SkeletalMeshSkeletonLocation001}
   &SphereLocation(MaskSpawnGroup=false, Module.Override Randomness=false, Sphere Radius=10, Surface Only=false)
   /SetVariables(Particles.SpawnTimeLerpColor:NiagaraFloat=&UniformRangedFloat(Minimum=0, Maximum=1)) {isSetParameters=true}
  ]
  particleUpdateScript [
   &ParticleState() {module=UpdateAge}
   &ScaleSpriteSize(Scale Factor=&Vector2DFromFloat(Value=&FloatFromCurve(FloatCurve=DataInterface() {propertyValues {Curve {keys [
          {interpMode=RCIM_Cubic, leaveTangent=1.4839, tangentMode=RCTM_Break, time=0, value=3}
          {arriveTangent=0.0927, interpMode=RCIM_Cubic, leaveTangent=0.0927, tangentMode=RCTM_User, time=1, value=6}
         ]}}}))) {module=SpriteSizeScale}
   /Niagara/Modules/Update/Material/DynamicMatParams(
    Module.Param0WriteEnabled=true
    Module.Param1WriteEnabled=true
    Module.Param2WriteEnabled=true
    Module.Param3WriteEnabled=true
    Param0Value:NiagaraFloat=$Particles.SpawnTimeLerpColor
    Param1Value:NiagaraFloat=&FloatFromCurve(FloatCurve=DataInterface() {propertyValues {Curve {keys [{interpMode=RCIM_Cubic, tangentMode=RCTM_Break, time=0, value=1.0113}, <at 0.3 0>]}}})
    Param2Value=1
    Param3Value=1
   )
   /Niagara/Modules/Update/Forces/DragForce(Drag=3)
   &Color(
    Color=$Particles.Initial.Color
    Scale Alpha=&FloatFromCurve(FloatCurve=DataInterface() {propertyValues {Curve {keys [{arriveTangent=0.0166, interpMode=RCIM_Cubic, leaveTangent=0.0166, tangentMode=RCTM_Break, time=0.0047, value=1.0198}, <at 1.0011 0.002>]}}})
    Scale Color=&VectorFromCurve(VectorCurve=DataInterface() {propertyValues {
       XCurve {keys [{interpMode=RCIM_Cubic, tangentMode=RCTM_Break, time=0.05, value=1.0067}, <at 0.9933 0.9143>]}
       YCurve {keys [{interpMode=RCIM_Cubic, tangentMode=RCTM_Break, time=0.0638, value=0.0911}, <at 0.995 0.6956>]}
       ZCurve {keys [{interpMode=RCIM_Cubic, tangentMode=RCTM_Break, time=0.0604, value=0.25}, <at 1.0018 0.9746>]}
      }})
   )
   &SolveForcesAndVelocity(Clamp Velocity=false, Limit Acceleration=false)
  ]
  rendererClasses [&NiagaraSpriteRendererProperties]
  renderers [{rendererClass=&NiagaraSpriteRendererProperties}]
 }
 Ribbons {
  emitterUpdateScript [
   &EmitterState(Enable Distance Culling=true, MaxDistance=2000, Module.bUseMaxDistance=true, Module.bUseMinDistance=false, Scalability Mode=::NewEnumerator1)
   &FN_SpawnRibbon_v2(Number of Ribbons=4, Spawn Mode=::NewEnumerator0, SpawnRate=20)
  ]
  particleSpawnScript [
   /Game/Effects/Niagara/Modules/Ribbons/FN_InitializeRibbon_v5(
    Color:LinearColor=<rgba 1 0.5 0.5 0.1>
    Enable Skeletal Position=true
    Lifetime=0.3
    Ribbon Width=28
    SkeletalMesh:NiagaraDataInterfaceSkeletalMesh=$User.SkeletalMesh
   )
   &AddVelocity(Velocity=&FN_FrontEndVector3Modifier(Gameplay=<0 0 0>, Front End=<0 0 0>))
   &DynamicMaterialParameters(
    Index 0 Param 1=1
    Index 0 Param 2=1
    Index 0 Param 3=0.1
    Index 0 Param 4=0.2
    Index 2 Param 1=0.2
    Index 2 Param 2=0.2
    Index 2 Param 3=0.2
    Index 3 Param 1=1
    Index 3 Param 2=1
    Index 3 Param 3=1
    Index 3 Param 4=1
    Module.Param0WriteEnabled=true
    Module.Param0WriteEnabled2=true
    Module.Param0WriteEnabled3=true
    Module.Param1WriteEnabled=true
    Module.Param1WriteEnabled2=true
    Module.Param1WriteEnabled3=true
    Module.Param2WriteEnabled=true
    Module.Param2WriteEnabled2=true
    Module.Param2WriteEnabled3=true
    Module.Param3WriteEnabled=true
    Module.Param3WriteEnabled2=true
    Module.Param3WriteEnabled3=true
    Write Parameter Index 2=true
    Write Parameter Index 3=true
   ) {module=DynamicMaterialParameters001}
  ]
  particleUpdateScript [
   &InheritSourceMovement(
    Applied Force Scale=&VectorFromFloat(Value=-1)
    Applied Position Scale=&FN_FrontEndVector3Modifier(Gameplay=<0.5 0.5 0.5>, Front End=<2 2 1>)
    ApplyForce=true
    ApplyForceFalloff=false
    ApplyPosition=true
    ApplyPositionFalloff=false
    Module.LimitForceSpeed=false
    Source Velocity:Vector3f=$System.SkydiveVelocity
   )
   &SolveForcesAndVelocity(Clamp Velocity=false, Limit Acceleration=false)
   &FN_UpdateRibbon(
    Enable Skeletal Position=true
    Initial Color:LinearColor=$Particles.Initial.Color
    Initial RibbonWidth:NiagaraFloat=$Particles.Initial.RibbonWidth
   )
   &DynamicMaterialParameters(
    Index 1 Param 1=&Add_Float(
     A=&MakeFloatFromInt(Int=$Particles.SpawnGroup)
     B=&Multiply_Float(A=-1.5, B=$Emitter.Age)
    )
    Index 1 Param 2=&FloatFromCurve(
     FloatCurve=DataInterface() {propertyValues {Curve {keys [{interpMode=RCIM_Cubic, tangentMode=RCTM_User, time=0, value=1}, {interpMode=RCIM_Cubic, tangentMode=RCTM_User, time=0.66, value=0}]}}}
     Scale Curve=0.15
    )
    Index 1 Param 3=1
    Index 1 Param 4=0.5
    Module.Param0WriteEnabled1=true
    Module.Param1WriteEnabled1=true
    Module.Param2WriteEnabled1=true
    Module.Param3WriteEnabled1=true
    Write Parameter Index 0=false
    Write Parameter Index 1=true
   )
   &ScaleRibbonWidth(Ribbon Width Scale=&FloatFromCurve(
     FloatCurve=DataInterface() {propertyValues {Curve {keys [{tangentMode=RCTM_User, time=0, value=0.9}, {tangentMode=RCTM_User, time=1, value=1}]}}}
     Scale Curve=0.5
    )) {module=ScaleRibbonWidth001}
   &Color(
    Color=&ColorFromCurve(ColorCurve=DataInterface() {propertyValues {
       AlphaCurve {keys [<at 0 1 RCIM_Cubic>,<at 0.9977 1 RCIM_Cubic>]}
       BlueCurve {keys [<at 0 0.3042 RCIM_Cubic>,<at 0.9978 0.6179>]}
       GreenCurve {keys [<at 0 0.2357 RCIM_Cubic>,<at 0.9978 0.2604>]}
       RedCurve {keys [<at 0 0.8229 RCIM_Cubic>,<at 0.9978 1>]}
      }})
    Scale Alpha=&FloatFromCurve(
     FloatCurve=DataInterface() {propertyValues {Curve {keys [
         {interpMode=RCIM_Cubic, tangentMode=RCTM_User, time=0, value=0}
         {interpMode=RCIM_Cubic, tangentMode=RCTM_User, time=0.15, value=1}
         {arriveTangent=-1.6667, interpMode=RCIM_Cubic, leaveTangent=-1.6667, time=0.4, value=0.25}
         {tangentMode=RCTM_User, time=0.75, value=0}
        ]}}}
     CurveIndex=/Game/Effects/Niagara/DynamicInputs/Math/FN_NormalizedParticleCount()
     Scale Curve=/Game/Effects/Niagara/DynamicInputs/FN_PickFloatFromSpawnGroupCutoff(
      Over Cutoff Float:NiagaraFloat=0.05
      Under Cutoff Float:NiagaraFloat=0.1
      SpawnGroup:NiagaraInt32=$Particles.SpawnGroup
      Cutoff:NiagaraInt32=2
     )
    )
    Scale Color=&VectorFromFloat(Value=1.5)
   )
  ]
  rendererClasses [&NiagaraRibbonRendererProperties]
  renderers [{rendererClass=&NiagaraRibbonRendererProperties}]
 }
 SmokeRing {
  emitterUpdateScript [
   &EmitterState(
    Inactive Response=::NewEnumerator1
    Life Cycle Mode=::NewEnumerator1
    Loop Duration=&RandomRangeFloat(Minimum=1, Maximum=3, Evaluation Type=::NewEnumerator1)
    Recalculate Duration Each Loop=true
    UseLoopDelay=true
   )
   &SpawnBurst_Instantaneous(Spawn Count=1, Use Spawn Probability=false)
  ]
  particleSpawnScript [
   &InitializeParticle(
    Color=<rgba 1 1 1 0.33>
    Lifetime=&RandomRangeFloat(Minimum=0.55, Maximum=0.75, Evaluation Type=::NewEnumerator1)
    Mesh Scale=<0.6 0.6 0.6>
    Module.Write Color=true
    Module.Write Lifetime=true
    Module.Write Mass=true
    Module.Write Material Random=true
    Module.Write Position=false
    Module.Write Scale=true
    Module.Write SpriteRotation=false
    Module.Write SpriteSize=true
    Position=<0 0 0>
    Sprite Size=&Vector2DFromFloat(Value=&FN_FrontEndFloatModifier(Gameplay=100, Front End=40))
   )
   &DynamicMaterialParameters(
    Index 0 Param 1=&RandomRangeFloat(Minimum=0, Maximum=1, Evaluation Type=::NewEnumerator1)
    Module.Param0WriteEnabled=true
    Module.Param1WriteEnabled=false
    Module.Param2WriteEnabled=false
    Module.Param3WriteEnabled=false
   )
   &SkeletalMeshLocation(
    Bone / Socket Index=4
    Bone Sampling Mode=::NewEnumerator4
    Inherit Velocity=true
    MaskSpawnGroup=false
    Sampled Position Offset=&TransformVector(Vector=<0 0 -150>, Destination Space=::Simulation)
    Skeletal Mesh=$User.SkeletalMesh
    VelocitySpeedLimit=false
   )
   &AddVelocity(
    Coordinate Space=::Simulation
    Velocity=&FN_FrontEndVector3Modifier(Gameplay=<0 0 0>, Front End=$System.SkydiveVelocityFE)
   )
  ]
  particleUpdateScript [
   &ParticleState()
   &ScaleSpriteSize(Scale Factor=&Vector2DFromFloat(Value=&FloatFromCurve(FloatCurve=DataInterface() {propertyValues {Curve {keys [
          {interpMode=RCIM_Cubic, leaveTangent=1.4839, tangentMode=RCTM_Break, time=0, value=3}
          {arriveTangent=0.0927, interpMode=RCIM_Cubic, leaveTangent=0.0927, tangentMode=RCTM_User, time=1, value=6}
         ]}}})))
   &ScaleColor(
    Scale Alpha=&FloatFromCurve(FloatCurve=DataInterface() {propertyValues {Curve {keys [<at 0 1 RCIM_Cubic>,<at 1 0 RCIM_Cubic>]}}})
    Scale RGB=&VectorFromCurve(VectorCurve=DataInterface() {propertyValues {
       XCurve {keys [<at 0.1 1 RCIM_Cubic>,<at 1 0.9521>]}
       YCurve {keys [<at 0.1 0.0104 RCIM_Cubic>,<at 1 0.6823>]}
       ZCurve {keys [<at 0.1 0.2659 RCIM_Cubic>,<at 1 1>]}
      }})
    ScaleA=true
    ScaleRGB=true
   )
   &DynamicMaterialParameters(
    Index 0 Param 2=&FloatFromCurve(FloatCurve=DataInterface() {propertyValues {Curve {keys [<at 0 1 RCIM_Cubic>,<at 0.3 0>]}}})
    Index 0 Param 3=1
    Index 0 Param 4=1
    Module.Param0WriteEnabled=false
    Module.Param1WriteEnabled=true
    Module.Param2WriteEnabled=true
    Module.Param3WriteEnabled=true
   ) {module=DynamicMaterialParameters001}
   &Drag(Drag=3, Use Linear Drag=true, Use Rotational Drag=false)
   &SolveForcesAndVelocity(Clamp Velocity=false, Limit Acceleration=false)
  ]
  rendererClasses [&NiagaraSpriteRendererProperties]
  renderers [{rendererClass=&NiagaraSpriteRendererProperties}]
 }
}
summary {
 systemName=NS_Contrail_Hearts
 emitters [
  {bEnabled=false, emitterName=Large_Heart, rendererClasses [&NiagaraSpriteRendererProperties]}
  {emitterName=Hearts_Hand, rendererClasses [&NiagaraSpriteRendererProperties]}
  {emitterName=SmokeRing, rendererClasses [&NiagaraSpriteRendererProperties]}
  {emitterName=Ribbons, rendererClasses [&NiagaraRibbonRendererProperties]}
 ]
 userVariables [
  User.bIsFrontEnd {type=NiagaraBool}
  User.bIsFrontEndPreview {type=NiagaraBool, defaultValue=true}
  User.ParaGlideLeanAlpha {type=NiagaraFloat, defaultValue=0, description=""}
  User.RibbonIDOffset {type=NiagaraInt32}
  User.SkeletalMesh {
   type=NiagaraDataInterfaceSkeletalMesh
   defaultValue {
    struct=NiagaraExt_VariableValue_DataInterface
    value {dataInterface=&self:NiagaraDataInterfaceSkeletalMesh_0}
   }
  }
 ]
}
